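# Translucent UI material, named as the background for gem hit feedback in Pilgrim widgets.
# It blends two color parameters, Color (default green) and Color2 (default magenta), through a linear interpolation.
# The LinearGradient function's U gradient drives the blend, giving a horizontal color ramp across the UV space.
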Material /PilgrimCore/Widgets/Materials/GemHitFeedbackMaterials/M_UI_Pilgrim_GemHitFeedback_BG
graph {
 properties {
  bUsedWithStaticMesh=false
  blendMode=BLEND_Translucent
  materialDomain=MD_UI
  pixelDepthOffsetMode=PDOM_Legacy
  usedShadingModels=Default Lit
 }
 nodes [LinearInterpolate(
   A=VectorParameter() {parameterName=Color, defaultValue=<rgba 0 1 0.1155 1>}.RGB
   B=VectorParameter() {parameterName=Color2, defaultValue=<rgba 0.9149 0 1 1>}.RGB
   Alpha=MaterialFunctionCall() {materialFunction=&LinearGradient}.UGradient
  )]
}
