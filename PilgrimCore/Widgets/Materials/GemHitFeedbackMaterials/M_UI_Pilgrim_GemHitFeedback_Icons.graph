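# Translucent UI material for Pilgrim gem hit feedback icons, drawing a glowing rounded icon shape procedurally.
# Layered MF_UI_Box calls build the fill, outline and glow masks, with GlowSize and a height parameter adjusting proportions.
# GlowColor and IconColor blend into the color, while ScaleAnim scales UVs about center and IconType shifts them vertically across three positions.
# The icon texture sample is present but unconnected.
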Material /PilgrimCore/Widgets/Materials/GemHitFeedbackMaterials/M_UI_Pilgrim_GemHitFeedback_Icons
graph {
 properties {
  bUsedWithStaticMesh=false
  blendMode=BLEND_Translucent
  materialDomain=MD_UI
  pixelDepthOffsetMode=PDOM_Legacy
  usedShadingModels=Default Lit
 }
 nodes [
  n1 = VectorParameter() {parameterName=GlowColor, defaultValue=<rgba 0.0373 0 1 1>}
  n2 = NamedRerouteUsage()
  NamedRerouteDeclaration(None=LinearInterpolate(
    A=LinearInterpolate(B=@n1.RGB, Alpha=@n2)
    B=VectorParameter() {parameterName=IconColor, defaultValue=<rgba 1 1 1 1>}.RGB
    Alpha=NamedRerouteUsage()
   )) {name=Color, nodeColor=<rgba 0.4194 0 1 1>}
  n3 = NamedRerouteUsage()
  n4 = Constant() {r=0.6}
  n5 = AppendVector(A=Constant() {r=1}, B=@n4)
  n6 = ScalarParameter() {defaultValue=0.05, parameterName=GlowSize}
  n7 = Subtract(A=Constant() {r=0.8}, B=@n6)
  n8 = Constant() {r=0.06}
  n9 = ScalarParameter() {defaultValue=-0.1, parameterName=Hightdjustment}
  n10 = Add(A=Constant() {r=0.4}, B=@n9)
  n11 = MaterialFunctionCall(UVs=NamedRerouteUsage(), Size=AppendVector(A=Subtract(A=@n7, B=@n8), B=@n10), CornerRoundness=@n10, OutlineThickness=Constant() {r=0.08}) {materialFunction=&MF_UI_Box}
  n12 = LinearInterpolate(
   B=Max(
    A=@n11.Outline
    B=MaterialFunctionCall(UVs=@n3, Size=@n5, CornerRoundness=@n4) {materialFunction=&MF_UI_Box}.Fill
   )
   Alpha=ScalarParameter() {defaultValue=0.5, parameterName=GlowOP}
  )
  n13 = Max(A=@n11.Fill, B=@n11.Fill)
  n14 = Add(A=@n4, B=@n9)
  n15 = Subtract(A=Constant() {r=1}, B=@n6)
  n16 = MaterialFunctionCall(UVs=NamedRerouteUsage(), Size=AppendVector(A=Subtract(A=@n15, B=@n8), B=@n14), CornerRoundness=@n14, OutlineThickness=@n8) {materialFunction=&MF_UI_Box}
  n17 = Max(A=@n11.Fill, B=Min(A=@n16.Outline, B=OneMinus(None=@n13)))
  NamedRerouteDeclaration(None=Max(A=@n12, B=NamedRerouteDeclaration(None=@n17) {name=IconSHape, nodeColor=<rgba 1 0 0.7031 1>})) {name=MainShape, nodeColor=<rgba 0.8516 0 1 1>}
  NamedRerouteDeclaration(None=TextureCoordinate()) {name=TextCoord, nodeColor=<rgba 1 0.0066 0 1>}
  NamedRerouteDeclaration(None=@n12) {name=GlowMask, nodeColor=<rgba 0.3828 1 0 1>}
  n18 = NamedRerouteUsage()
  n19 = LinearInterpolate(Alpha=ScalarParameter() {parameterName=ScaleAnim}) {constA=1, constB=1.1}
  n20 = Constant()
  n21 = Constant() {r=0.18}
  n22 = MaterialFunctionCall(A=@n21, B=Constant(), C=Multiply(A=@n21) {constB=-1}, Alpha=ScalarParameter() {parameterName=IconType}) {materialFunction=&Lerp_3Color}
  NamedRerouteDeclaration(None=Add(
    A=MaterialFunctionCall(UVs=@n18, Texture Scale=@n19) {materialFunction=&ScaleUVsByCenter}.UVs
    B=AppendVector(A=@n20, B=@n22.Result)
   )) {name=IconType, nodeColor=<rgba 1 0 0.2578 1>}
  NamedRerouteUsage()
  NamedRerouteUsage()
  Saturate(None=@n16.Outline)
  TextureSampleParameter2D() {parameterName=Icon Texture, texture=/PilgrimCore/Widgets/Textures/GemHitFeedback/T_Pilgrim_HitFeedback_Perfect}
 ]
}
